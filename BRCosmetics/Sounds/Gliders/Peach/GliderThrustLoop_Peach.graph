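# Looping thrust sound cue for the Peach cosmetic glider, mixing four flight layers: wind, chimes, gallop and trot.
# The AnyDirection parameter scales the volume of wind, chimes and trot and also raises wind pitch, while Forwards alone drives gallop volume.
# An Enveloper slightly eases down the chimes over their first seconds.
# Shared glider thrust attenuation, sound class and audio property sheet govern playback.
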SoundCue /BRCosmetics/Sounds/Gliders/Peach/GliderThrustLoop_Peach
root=Mixer_0
nodes [
 {
  id=Enveloper_0
  children [ModulatorContinuous_1]
  properties {pitchCurve {editorCurveData {keys [{time=0}]}}, volumeCurve {editorCurveData {keys [<at 0 1>,<at 3 0.85>]}}}
 }
 {id=Mixer_0, children [ModulatorContinuous_0, Enveloper_0, ModulatorContinuous_2, ModulatorContinuous_3], properties {inputVolume [1,1,1,1]}}
 {
  id=ModulatorContinuous_0
  children [WavePlayer_0]
  properties {
   pitchModulationParams {default=0, maxOutput=1.2, minOutput=1, parameterName=AnyDirection}
   volumeModulationParams {default=0, minOutput=0.2, parameterName=AnyDirection}
  }
 }
 {
  id=ModulatorContinuous_1
  children [WavePlayer_1]
  properties {
   pitchModulationParams {default=0, minOutput=1, parameterName=AnyDirection}
   volumeModulationParams {default=0, minOutput=0.6, parameterName=AnyDirection}
  }
 }
 {
  id=ModulatorContinuous_2
  children [WavePlayer_2]
  properties {pitchModulationParams {default=0, minOutput=1, parameterName=Forwards}, volumeModulationParams {default=0, parameterName=Forwards}}
 }
 {
  id=ModulatorContinuous_3
  children [WavePlayer_3]
  properties {
   pitchModulationParams {default=0, minOutput=1, parameterName=AnyDirection}
   volumeModulationParams {default=0, maxOutput=0.8, minOutput=0.3, parameterName=AnyDirection}
  }
 }
 {id=WavePlayer_0, properties {bLooping=true, soundWaveAssetPtr=&/Glider_Peach_Flight_Wind_LP}}
 {id=WavePlayer_1, properties {bLooping=true, soundWaveAssetPtr=&/Glider_Peach_Flight_Chimes_LP}}
 {id=WavePlayer_2, properties {bLooping=true, soundWaveAssetPtr=&/Glider_Peach_Flight_Gallop_LP}}
 {id=WavePlayer_3, properties {bLooping=true, soundWaveAssetPtr=&/Glider_Peach_Flight_Trot_LP}}
]
properties {
 attenuationSettings=&Fort_Player_Glider_Thrust_Athena_Att
 audioPropertiesSheet=&APS_Cosmetic_Glider_Loop
 maxDistance=12500
 soundClassObject=&Fort_Gliders_Land
 volumeMultiplier=0.9
 concurrencySet [null]
}
